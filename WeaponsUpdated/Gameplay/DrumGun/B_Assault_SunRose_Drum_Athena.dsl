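; B_Assault_SunRose_Drum_Athena is the weapon Blueprint for the SunRose drum assault rifle, and it shows a rail on the gun when an optic mod is attached.
; On BeginPlay it binds an event to the mod-added-to-weapon notification.
; When the added mod is in the optic slot and is not the default iron sights, it sets "Hide Element 01" to zero on the weapon mesh's dynamic material to reveal the rail.
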(event EventBeginPlay
  (|CallParentFunction)
  (WeaponMods|AssignOnModAddedtoWeapon (delegate Custom|OnModAddedToWeapon_Event)))

(event Custom|OnModAddedToWeapon_Event
  (params (Mod :type Fort%20Weapon%20Mod%20Item%20Definition%20Object%20Reference) (Weapon :type Fort%20Weapon%20Object%20Reference))
  ;[1 1 1] When an Optic Mod is applied to the weapon, display the Rail
  (if (and (!= Mod "/WeaponMods/Items/WeaponMods/Optics/IronSight/WMID_Default_IronSights.WMID_Default_IronSights") (== (Variables|/Script/FortniteGame.FortWeaponModItemDefinition|GetModSlot Mod) (Variables|Default|GetOpticModTag)))
    (bind _returnvalue (Fortnite|GetDynamicMaterialInstance (Variables|Visual|GetWeaponMesh) 1))
    (Rendering|Material|SetScalarParameterValue _returnvalue "Hide Element 01" 0.0 :!class /Script/Engine.MaterialInstanceDynamic)))

(fn ConstructionScript
  (|CallParentFunction))
