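# Sound cue that plays when the SunRose DrumGun is dropped, reusing generic heavy rifle drop sounds.
# A Random node picks among four equally weighted wave variations and feeds a root Modulator at its default settings for slight variation.
# The cue uses weapon drop attenuation and the weapon foley sound class at reduced volume, with concurrency overridden and the debug flag left on.
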SoundCue /WeaponsUpdated/Gameplay/DrumGun/Sounds/SunRose_DrumGun_Drop_Cue
root=Modulator_1
nodes [
 {id=Modulator_1, children [Random_0]}
 {id=Random_0, children [WavePlayer_0,WavePlayer_2,WavePlayer_3,WavePlayer_4], properties {weights [1,1,1,1]}}
 {id=WavePlayer_0, properties {soundWaveAssetPtr=/Game/Sounds/Fort_UI_Sounds/Gameplay/Weapon_Generic_Drop_Rifle_Hv_01}}
 {id=WavePlayer_2, properties {soundWaveAssetPtr=/Game/Sounds/Fort_UI_Sounds/Gameplay/Weapon_Generic_Drop_Rifle_Hv_02}}
 {id=WavePlayer_3, properties {soundWaveAssetPtr=/Game/Sounds/Fort_UI_Sounds/Gameplay/Weapon_Generic_Drop_Rifle_Hv_03}}
 {id=WavePlayer_4, properties {soundWaveAssetPtr=/Game/Sounds/Fort_UI_Sounds/Gameplay/Weapon_Generic_Drop_Rifle_Hv_04}}
]
properties {
 attenuationSettings=&Fort_Weapon_Drop_Att
 bDebug=true
 bOverrideConcurrency=true
 maxDistance=3000
 soundClassObject=&Fort_WeaponFoley
 volumeMultiplier=0.501
}
