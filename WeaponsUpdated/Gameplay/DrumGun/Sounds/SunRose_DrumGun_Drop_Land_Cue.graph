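# Sound cue for a SunRose Drum Gun landing when dropped, built from shared generic assault rifle drop sounds.
# A Random node picks one of four equally weighted waves and feeds a default Modulator for pitch and volume variation.
# It plays at reduced volume through the weapon foley sound class and weapon drop attenuation, overrides concurrency, and has its debug flag left on.
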SoundCue /WeaponsUpdated/Gameplay/DrumGun/Sounds/SunRose_DrumGun_Drop_Land_Cue
root=Modulator_0
nodes [
 {id=Modulator_0, children [Random_1]}
 {id=Random_1, children [WavePlayer_1,WavePlayer_3,WavePlayer_4,WavePlayer_5], properties {weights [1,1,1,1]}}
 {id=WavePlayer_1, properties {soundWaveAssetPtr=/Game/Sounds/Fort_UI_Sounds/Gameplay/Weapon_Generic_Drop_AR_04}}
 {id=WavePlayer_3, properties {soundWaveAssetPtr=/Game/Sounds/Fort_UI_Sounds/Gameplay/Weapon_Generic_Drop_AR_01}}
 {id=WavePlayer_4, properties {soundWaveAssetPtr=/Game/Sounds/Fort_UI_Sounds/Gameplay/Weapon_Generic_Drop_AR_02}}
 {id=WavePlayer_5, properties {soundWaveAssetPtr=/Game/Sounds/Fort_UI_Sounds/Gameplay/Weapon_Generic_Drop_AR_03}}
]
properties {
 attenuationSettings=&Fort_Weapon_Drop_Att
 bDebug=true
 bOverrideConcurrency=true
 maxDistance=3000
 soundClassObject=&Fort_WeaponFoley
 volumeMultiplier=0.501
}
