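# Sound cue for picking up the SunRose DrumGun, reusing the generic rifle pickup sound as weapon foley.
# A quality level node routes the first two levels and the last two to separate mixers, each boosting its own player of the same Rifle_Pickup wave, with the second input empty.
# It borrows weapon reload attenuation, lowers overall cue volume, overrides concurrency and has the debug flag on.
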SoundCue /WeaponsUpdated/Gameplay/DrumGun/Sounds/SunRose_DrumGun_Pickup_Cue
root=QualityLevel_0
nodes [
 {id=Mixer_0, children [WavePlayer_6,null], properties {inputVolume [1.2,1]}}
 {id=Mixer_1, children [WavePlayer_3,null], properties {inputVolume [1.2,1]}}
 {id=QualityLevel_0, children [Mixer_1,Mixer_1,Mixer_0,Mixer_0]}
 {id=WavePlayer_3, properties {soundWaveAssetPtr=/Game/Sounds/Fort_GamePlay_Sounds/Loot/Rifle_Pickup}}
 {id=WavePlayer_6, properties {soundWaveAssetPtr=&Rifle_Pickup}}
]
properties {
 attenuationSettings=&Fort_Weapon_Reload_Att
 bDebug=true
 bOverrideConcurrency=true
 maxDistance=2250
 soundClassObject=&Fort_WeaponFoley
 volumeMultiplier=0.451
}
