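# Sound cue for the insert step of the SunRose DrumGun reload foley.
# A Random node picks one of three equally weighted insert waves and feeds a Modulator at default settings for slight pitch and volume variation.
# It uses the shared weapon reload attenuation and weapon foley sound class, with reduced volume and overridden concurrency.
# The debug flag is left enabled on the cue.
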SoundCue /WeaponsUpdated/Gameplay/DrumGun/Sounds/SunRose_DrumGun_Reload_Insert_Cue
root=Modulator_0
nodes [
 {id=Modulator_0, children [Random_0]}
 {id=Random_0, children [WavePlayer_1,WavePlayer_2,WavePlayer_3], properties {weights [1,1,1]}}
 {id=WavePlayer_1, properties {soundWaveAssetPtr=&/SunRose_DrumGun_Reload_Insert_01}}
 {id=WavePlayer_2, properties {soundWaveAssetPtr=&/SunRose_DrumGun_Reload_Insert_02}}
 {id=WavePlayer_3, properties {soundWaveAssetPtr=&/SunRose_DrumGun_Reload_Insert_03}}
]
properties {
 attenuationSettings=&Fort_Weapon_Reload_Att
 bDebug=true
 bOverrideConcurrency=true
 maxDistance=2250
 soundClassObject=&Fort_WeaponFoley
 volumeMultiplier=0.451
}
